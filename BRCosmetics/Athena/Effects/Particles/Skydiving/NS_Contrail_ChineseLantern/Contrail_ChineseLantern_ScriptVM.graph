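# Niagara skydiving contrail cosmetic that trails ribbons and Chinese lanterns behind the player.
# The Empty emitter spawns ribbons by distance, locks particles to skeletal mesh bones per spawn group, and applies curl noise, twist and width curves.
# The Lantern emitter spawns glowing mesh and sprite lanterns on a ring, adding vortex spin, scale curves and late-life upward acceleration.
# Dive state and front-end modifiers retune motion and spawning.
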NiagaraSystem /BRCosmetics/Athena/Effects/Particles/Skydiving/NS_Contrail_ChineseLantern/Contrail_ChineseLantern_ScriptVM
dependencies {
 usedDataInterfaces [&NiagaraDataInterfaceCurve]
 usedDynamicInputs []
 usedModules [
  /Niagara/Modules/System/SystemLifeCycle
  &SystemState
  /Game/Effects/Niagara/Modules/Skydiving/FN_DivingAttributes/v6/FN_DivingAttributes
 ]
}
emitters {
 Empty {
  emitterUpdateScript [
   &EmitterState()
   /Game/Effects/Niagara/Modules/Spawn/FN_SpawnByEngineDistance() {enabled=false}
   &self:FN_SpawnByDistance_Modified(FE TimeStep=10, Multiplier=4, Step=100) {module=FN_SpawnByDistance_Modified}
  ]
  particleSpawnScript [
   &InitializeRibbon(
    Lifetime=&FN_FrontEndFloatModifier(Gameplay=0.3, Front End=0.8)
    Module.Write Color=true
    Module.Write Lifetime=true
    Module.Write Mass=true
    Module.Write Material Random=true
    Module.Write Position=false
    Module.Write RibbonTwist=false
    Module.Write RibbonWidth=true
   )
   /SetVariables(Particles.Stream:NiagaraInt32=&SpawnGroup()) {isSetParameters=true}
   &FN_SetSpawnGroup()
   &SkeletalMeshLocation(
    Apply To Particle Position=false
    Bone / Socket Index=$Particles.SpawnGroup
    Bone Sampling Mode=::NewEnumerator4
    Inherit Velocity=false
    Interpolate Mesh Sampling=false
    MaskSpawnGroup=false
    Skeletal Mesh=$User.SkeletalMesh
   ) {module=SkeletalMeshLocation001}
   &AddVelocity(Velocity=&FN_FrontEndVector3Modifier(Gameplay=<0 0 0>, Front End=<0 0 300>))
  ]
  particleUpdateScript [
   &ParticleState()
   /Game/Effects/Niagara/Modules/Skydiving/FN_DivingAttributes/FN_DivingAttributes(Bone Index=2)
   /Niagara/Modules/Update/Forces/CurlNoiseForce(
    Noise Frequency:NiagaraFloat=&Lerp_Float(A=0.005, B=0.01, Alpha=$Particles.Dive)
    Noise Strength=/Game/Effects/Niagara/DynamicInputs/Skydiving/SkydiveFloatModifier_v2(Float:NiagaraFloat=&FN_FrontEndFloatModifier(Gameplay=500, Front End=300), Dive:NiagaraFloat=600)
    Pan Noise=true
    Pan Noise Field=&MakeVector(
     X=&Multiply_Float(A=&MakeFloatFromInt(Int=$Particles.SpawnGroup), B=0.5)
     Y=0
     Z=&Lerp_Float(A=2, B=0, Alpha=$Particles.Dive)
    )
   )
   &SkeletalMeshLocation(
    Apply To Particle Position=false
    Bone / Socket Index=$Particles.SpawnGroup
    Bone Sampling Mode=::NewEnumerator4
    Inherit Velocity=false
    Interpolate Mesh Sampling=false
    MaskSpawnGroup=false
    Skeletal Mesh=$User.SkeletalMesh
   )
   &InheritSourceMovement(
    Applied Position Scale=&VectorFromFloat(Value=0.45)
    ApplyForce=false
    ApplyPosition=true
    ApplyPositionFalloff=false
    Source Velocity:Vector3f=$Output.SkeletalMeshLocation.SampledVelocity
   )
   &SolveForcesAndVelocity(Clamp Velocity=false, Limit Acceleration=false)
   &self:Lock_Particles(
    Count:NiagaraInt32=$Emitter.Count
    Locked Position:Vector3f=$Output.SkeletalMeshLocation.SampledPosition
    Stream:NiagaraInt32=$Particles.Stream
   ) {module=Lock_Particles}
   /Game/Effects/Niagara/Modules/Ribbons/FN_RibbonTwist(Multiplier=3, TwistSpeed:NiagaraFloat=&Lerp_Float(A=10, B=10, Alpha=$Particles.Dive))
   /Game/Effects/Niagara/Modules/Ribbons/FN_RibbonWidthScale(Ribbon Width:NiagaraFloat=&FloatFromCurve(FloatCurve=DataInterface() {propertyValues {Curve {keys [
         <at -0.0035 0.5904>
         <at 0.0329 0.7122>
         {arriveTangent=7.0279, interpMode=RCIM_Cubic, leaveTangent=7.0279, tangentMode=RCTM_User, time=0.0566, value=1.0189}
         <at 0.9927 1.0189>
        ]}}}))
  ]
  rendererClasses [&NiagaraSpriteRendererProperties, &NiagaraRibbonRendererProperties]
  renderers [
   {rendererClass=&NiagaraSpriteRendererProperties}
   {rendererClass=&NiagaraRibbonRendererProperties, rendererIndex=1}
  ]
 }
 Lantern {
  emitterUpdateScript [
   &EmitterScalability(
    MaxDistance=7500
    Module.bApplySpawnCountScale=false
    Module.bApplySpawnCountScaleByDistanceFraction=false
    Module.bUseMaxDistance=true
    Module.bUseMinDistance=false
    SpawnCountScaleByDistanceFraction=DataInterface() {propertyValues {Curve {keys [<at -0.0008 1.0263>,<at 0.99 1>,<at 1 0.5>]}}}
   ) {enabled=false}
   &EmitterLifeCycle(NextLoopDuration=1) {enabled=false}
   &EmitterState()
   &SpawnRate(SpawnRate=&FN_FrontEndFloatModifier(Gameplay=4, Front End=2), Use Spawn Probability=false)
  ]
  particleSpawnScript [
   /Niagara/Modules/Spawn/Initialization/InitializeParticle(
    Color=<rgba 2 0.5 0 1>
    Lifetime=2
    Mesh Scale=&VectorFromFloat(Value=&UniformRangedFloat(Minimum=0.5, Maximum=0.3))
    Module.Write Color=true
    Module.Write Lifetime=true
    Module.Write Mass=true
    Module.Write Material Random=true
    Module.Write Position=false
    Module.Write Scale=true
    Module.Write SpriteRotation=false
    Module.Write SpriteSize=true
    Position=&SimulationPosition()
    Sprite Size=<50 50>
   )
   /Niagara/Modules/Spawn/Orientation/V2/InitialMeshOrientation(Orientation Vector=<0 0 1>, Use Orientation Vector=false, Use Rotation Vector=false) {module=InitialMeshRotation}
   &CylinderLocation(
    Cylinder Height=0
    Cylinder Radius=75
    Lathe Profile=DataInterface() {propertyValues {Curve {keys []}}}
    MaskSpawnGroup=false
    Module.Enable Lathe Profile=false
    Module.Override Randomness=false
    Module.Surface Only=true
    Offset=<0 0 75>
    Surface Only Band Thickness=0
    Use Endcaps In Surface Only Mode=false
   )
   &DynamicMaterialParameters(
    Index 0 Param 1=-1
    Index 0 Param 4=&UniformRangedFloat(Minimum=0, Maximum=2)
    Module.Param0WriteEnabled=true
    Module.Param1WriteEnabled=true
    Module.Param2WriteEnabled=true
    Module.Param3WriteEnabled=true
   ) {module=DynamicMatParams}
   /SetVariables(
    Particles.InitialMeshRotationRate:NiagaraFloat=&UniformRangedFloat(Minimum=36, Maximum=300)
    Particles.VortexVelocityRate:NiagaraFloat=&UniformRangedFloat(Minimum=50, Maximum=150)
   ) {isSetParameters=true}
   &InheritVelocity(Module.Limit Speed=true, Velocity Scale=&VectorFromFloat(Value=-0.0145))
   &AddVelocity(Velocity=&Multiply_Vector(A=<1 1 0>, B=<1 1 0>))
  ]
  particleUpdateScript [
   &FN_DivingAttributes(Bone Index=2)
   &ParticleState() {module=UpdateAge}
   /Niagara/Modules/Update/Orientation/V2/UpdateMeshOrientation(Rotation Vector:Vector3f=<0 0 1>) {module=MeshRotationRate}
   /Niagara/Modules/Update/Velocity/VortexVelocity(
    Module.Use Influence Falloff=false
    Velocity Amount:NiagaraFloat=$Particles.VortexVelocityRate
    Vortex Origin:Vector3f=/Niagara/DynamicInputs/Helpers/SimulationPosition()
   )
   &ScaleMeshSize(Scale Factor=&VectorFromFloat(Value=&FloatFromCurve(FloatCurve=DataInterface() {propertyValues {Curve {keys [<at -0.0046 0>,<at 0.09 1.4>,<at 0.2 1>,<at 0.9082 1.0219>,<at 1.0011 0.002>]}}}))) {module=MeshSizeScale}
   &AccelerationForce(Acceleration=&MakeVector(
     X=0
     Y=0
     Z=&FloatFromCurve(
      FloatCurve=DataInterface() {propertyValues {Curve {keys [<at 0.5 -0.0134 RCIM_Cubic>, {arriveTangent=3.0618, interpMode=RCIM_Cubic, leaveTangent=3.0618, tangentMode=RCTM_User, time=1, value=1}]}}}
      Scale Curve=&SkydiveFloatModifier_v2(Float:NiagaraFloat=5000, Dive:NiagaraFloat=20000)
     )
    ))
   &SolveForcesAndVelocity(Clamp Velocity=false, Limit Acceleration=false)
   &DynamicMaterialParameters(
    Index 0 Param 1=-10
    Index 0 Param 3=&FloatFromCurve(FloatCurve=DataInterface() {propertyValues {Curve {keys [<at -0.0026 -0.0134>,<at 1.0011 0.002>]}}}, Scale Curve=100)
    Index 0 Param 4=$Particles.Random
    Module.Param0WriteEnabled=true
    Module.Param1WriteEnabled=true
    Module.Param2WriteEnabled=true
    Module.Param3WriteEnabled=true
   ) {module=DynamicMatParams001}
   /SetVariables(Particles.LanternGlow:LinearColor=<rgba 1 0.1006 0 0.3>) {isSetParameters=true}
  ]
  rendererClasses [&NiagaraMeshRendererProperties, &NiagaraSpriteRendererProperties]
  renderers [
   {rendererClass=&NiagaraMeshRendererProperties, rendererIndex=0}
   {rendererClass=&NiagaraSpriteRendererProperties, rendererIndex=1}
  ]
 }
}
summary {
 systemName=Contrail_ChineseLantern_ScriptVM
 emitters [
  {emitterName=Lantern, rendererClasses [&NiagaraMeshRendererProperties, &NiagaraSpriteRendererProperties]}
  {emitterName=Empty, rendererClasses [&NiagaraSpriteRendererProperties, &NiagaraRibbonRendererProperties]}
 ]
 userVariables [
  User.bIsFrontEnd {type=NiagaraBool}
  User.bIsFrontEndPreview {type=NiagaraBool}
  User.FEVec {type=Vector3f, description="", defaultValue {struct=Vector3f, value=<0 0 0>}}
  User.FrontendPreviewVector {type=Vector3f, description="", defaultValue {struct=Vector3f, value=<0 0 0>}}
  User.PlayerSpeed {type=NiagaraFloat, defaultValue=0, description=""}
  User.RibbonIDOffset {type=NiagaraInt32}
  User.SkeletalMesh {
   type=NiagaraDataInterfaceSkeletalMesh
   defaultValue {
    struct=NiagaraExt_VariableValue_DataInterface
    value {dataInterface=&self:NiagaraDataInterfaceSkeletalMesh_0}
   }
  }
 ]
}
